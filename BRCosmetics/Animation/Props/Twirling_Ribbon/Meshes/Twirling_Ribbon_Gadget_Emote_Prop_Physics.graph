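# Physics asset for the Twirling Ribbon gadget emote prop, defining simplified collision bodies for its skeletal mesh.
# A long, thin capsule on the stick base joint covers the stick, and a tiny capsule sits on the root bone, well offset from its origin.
# A single constraint joins the two bodies with equal swing and twist angle limits.
# No bodies are defined for the ribbon itself.
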PhysicsAsset /BRCosmetics/Animation/Props/Twirling_Ribbon/Meshes/Twirling_Ribbon_Gadget_Emote_Prop_Physics
bodies {
 _Stick_Base_Jnt [{shapeType=Capsule, length=35.9914, radius=1.6337, center=<0.0484 35.6261 0.0607>, rotation=<rot 0 -179.9981 -89.9877>}]
 root [{shapeType=Capsule, length=0.505, radius=0.505, center=<0.0575 69.8587 -169.7327>, rotation=<rot 0.0034 -90 0.0164>}]
}
constraints [{bone1Name=_Stick_Base_Jnt, bone2Name=root, swing1LimitDegrees=45, swing2LimitDegrees=45, twistLimitDegrees=45}]
